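# Unlit, translucent, two-sided Niagara material for stylized wind stripe effects on the VinderTech GasChin glider.
# Three oppositely or differently panned caustic textures are blended, then masked by a V-gradient times its inverse to soften both edges.
# Particle alpha, depth fade and view-angle fade further shape that mask for opacity.
# Particle color RGB feeds an emissive-adjustment function for tint and brightness.
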Material /BRCosmetics/Gadgets/Assets/VinderTech_GliderChute/Glider_MuffinLadle_GasChin/FX/M_Wind_Stripes
graph {
 properties {
  bUsedWithNiagaraMeshParticles=true
  bUsedWithNiagaraRibbons=true
  bUsedWithNiagaraSprites=true
  blendMode=BLEND_Translucent
  pixelDepthOffsetMode=PDOM_Legacy
  previewMesh=&/SM_Torus_Wind_Motion
  shadingModel=MSM_Unlit
  twoSided=true
  usedShadingModels=Unlit
 }
 nodes [
  n1 = ParticleColor()
  MaterialFunctionCall(Emissive=@n1.RGB) {materialFunction=&MF_FX_EmissiveAdjustments}
  n2 = Panner(Coordinate=TextureCoordinate() {vTiling=3}) {speedX=-2}
  n3 = TextureSample(UVs=Panner(Coordinate=TextureCoordinate() {vTiling=3}) {speedX=2}) {samplerType=SAMPLERTYPE_LinearColor, texture=/BRCosmetics/Effects/Fort_Effects/Textures/Wind/T_Stylized_Wind_Caustics_3}
  n4 = TextureSample(UVs=Panner(Coordinate=TextureCoordinate() {vTiling=2}) {speedX=3}) {texture=/BRCosmetics/Effects/Fort_Effects/Textures/Wind/T_Stylized_Wind_Caustics}
  n5 = ComponentMask(None=TextureCoordinate()) {g=true}
  n6 = Multiply(
   A=LinearInterpolate(
    A=TextureSample(UVs=@n2) {texture=/BRCosmetics/Effects/Fort_Effects/Textures/Wind/T_Stylized_Wind_Caustics_2}.RGB
    B=LinearInterpolate(A=@n3.RGB, B=@n4.RGB) {constAlpha=0.4}
   ) {constAlpha=0.6}
   B=Multiply(A=@n5, B=OneMinus(None=@n5))
  )
  Multiply(
   A=DepthFade(Opacity=Multiply(A=@n1.A, B=@n6)) {fadeDistanceDefault=5}
   B=MaterialFunctionCall("FadeSensitivity (optional)"=Constant() {r=1}) {materialFunction=&FadeBasedOnViewAngle}.Result
  )
 ]
}
