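# Niagara idle effect for the MuffinLadle GasChin glider, attached to its skeletal mesh.
# Smoke, sparks, steam and flickering glow lights use partitioned bone offsets from a shared position array, alongside random-socket electrical bursts, anim-trail ribbons and a speed-motion mesh.
# System timelines drive spawn probability, visibility and alpha across emitters.
# Owner movement bends or carries particles, and front-end modifiers dim glows, redirect trails and disable steam in menus.
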NiagaraSystem /BRCosmetics/Gadgets/Assets/VinderTech_GliderChute/Glider_MuffinLadle_GasChin/FX/NS_Glider_MuffinLadle_GasChin_Idle
dependencies {
 usedDataInterfaces [
  &NiagaraDataInterfaceCurve
  &NiagaraDataInterfaceArrayFloat3
  &NiagaraDataInterfaceVector2DCurve
  &NiagaraDataInterfaceColorCurve
  &NiagaraDataInterfaceArrayFloat
  &NiagaraDataInterfaceVectorCurve
  &NiagaraDataInterfaceArrayPosition
 ]
 usedDynamicInputs [
  &RemapRange
  &MakeIntFromFloat
  &NormalizeVector
  &Add_Vector
  &FNTimeOfDayScalar
 ]
 usedModules [
  &self:SystemSpawnScript.SystemScriptSource.SystemScriptGraph.NiagaraNodeAssignment_0.SetVariables_#1
  &SystemState
  &Timeline
  &FN_Integrator
  &InterpolateOverTime
  &self:SystemSpawnScript.SystemScriptSource.SystemScriptGraph.NiagaraNodeAssignment_2.SetVariables_#2
  &self:SystemSpawnScript.SystemScriptSource.SystemScriptGraph.NiagaraNodeAssignment_1.SetVariables_#3
  &self:SystemSpawnScript.SystemScriptSource.SystemScriptGraph.NiagaraNodeAssignment_3.SetVariables_#4
 ]
}
emitters {
 ElectricalSparksBurst {
  emitterUpdateScript [
   /SetVariables(Emitter.RandomSockets:NiagaraInt32=&RandomRangeInt(Minimum=4, Maximum=7)) {isSetParameters=true}
   &EmitterState(
    Life Cycle Mode=::NewEnumerator1
    Loop Delay=&RandomRangeFloat(Minimum=0, Maximum=0.5, Evaluation Type=::NewEnumerator1)
    Loop Duration=&RandomRangeFloat(Minimum=0.5, Maximum=3, Evaluation Type=::NewEnumerator1)
    UseLoopDelay=true
   )
   &SpawnBurst_Instantaneous(
    Spawn Count=&RandomRangeInt(Minimum=5, Maximum=7)
    Use Spawn Probability=false
    UseLoopCountLimit=false
   )
   &SpawnBurst_Instantaneous(Spawn Group=1, Use Spawn Probability=true, UseLoopCountLimit=false) {module=SpawnBurst_Instantaneous001}
  ]
  particleSpawnScript [
   /SetVariables(Particles.VisibilityTag=$Emitter.SpawnGroup) {isSetParameters=true}
   /SetVariables(Particles.PerParticleRandom:NiagaraFloat=&FloatFromCurve(
     FloatCurve=DataInterface() {propertyValues {Curve {keys [
         {arriveTangent=4.1785, interpMode=RCIM_Cubic, tangentMode=RCTM_Break, time=0, value=0}
         {arriveTangent=3.1227, interpMode=RCIM_Cubic, leaveTangent=0.0224, tangentMode=RCTM_Break, time=1, value=1}
        ]}}}
     CurveIndex=&RandomRangeFloat(Minimum=0, Maximum=1)
    )) {isSetParameters=true}
   &InitializeParticle(
    Color Mode=::NewEnumerator0
    Lifetime=&RandomRangeFloat(Minimum=0.5, Maximum=1)
    Position Offset Coordinate Space=::Local
    UsePositionOffset=true
    Write Material Random=true
   )
   &SkeletalMeshLocation(
    Bone / Socket Index=$Emitter.RandomSockets
    Interpolate Mesh Sampling=false
    MaskSpawnGroup=false
    Mesh Sampling Type=::NewEnumerator5
    Skeletal Mesh=$System.Skeletal Mesh
    Socket Sampling Mode=::NewEnumerator5
   )
   &AddVelocity(
    Additional Quaternion Rotation=false
    Cone Angle=&RandomRangeFloat(Minimum=25, Maximum=90)
    Module.UseConeFalloff=false
    Module.UseVelDistribution=false
    Override Seed=false
    Rotation Mode=::NewEnumerator3
    Rotation Quaternion:Quat4f=$Particles.SkeletalMeshLocation.SampledRotation
    Velocity Mode=::NewEnumerator2
    Velocity Speed=&MaskFloatBySpawnGroup(Masked Float=&RandomRangeFloat(Minimum=150, Maximum=500), Passthrough Float=0)
   )
   &AddVelocity(
    Rotation Coordinate Space=::World
    Velocity=<0 0 1>
    Velocity Speed Scale=&RandomRangeFloat(Minimum=10, Maximum=40)
   ) {module=AddVelocity001}
   /SetVariables(
    Particles.FlareSpriteSize:Vector2f=<20 20>
    Particles.GlowSpriteSize:Vector2f=&Multiply_Vector2D_ByFloat(Float=1, Vector2D=&Lerp_Vector2(A=<5 10>, B=<10 20>, Alpha=$Particles.PerParticleRandom))
    Particles.SparkSpriteSize:Vector2f=&Multiply_Vector2D_ByFloat(Float=1, Vector2D=&Lerp_Vector2(A=<1 2>, B=<2 7>, Alpha=$Particles.PerParticleRandom))
   ) {isSetParameters=true}
   &CameraOffset(Camera Offset Amount=&MaskFloatBySpawnGroup(Masked Float=0, Passthrough Float=5))
  ]
  particleUpdateScript [
   &ParticleState()
   &SkeletalMeshLocation(
    Bone / Socket Index=$Emitter.RandomSockets
    MaskSpawnGroup=false
    Mesh Sampling Type=::NewEnumerator5
    Position Sampling=::NewEnumerator2
    Skeletal Mesh=$System.Skeletal Mesh
    Socket Sampling Mode=::NewEnumerator5
   ) {module=SkeletalMeshLocation001}
   /SetVariables(Particles.Position_LightGlow:NiagaraPosition=$Particles.SkeletalMeshLocation001.SampledPosition) {isSetParameters=true}
   /SetVariables(Particles.SparkSpriteColor:LinearColor=&MultiplyLinearColor(
     Linear Color=&ColorFromCurve(ColorCurve=DataInterface() {propertyValues {
        AlphaCurve {keys [<at 0.01 1>,<at 0.9198 1 RCIM_Cubic>,<at 1 0>]}
        BlueCurve {keys [<at 0.0115 2.4844>,<at 0.068 0>,<at 0.1437 0>,<at 0.7101 0>,<at 0.9394 0>]}
        GreenCurve {keys [<at 0.0115 2.8625>,<at 0.068 2.2>,<at 0.1437 0.5>,<at 0.7101 0.1833>,<at 0.9394 0.0762>]}
        RedCurve {keys [<at 0.0115 3>,<at 0.068 3>,<at 0.1437 2>,<at 0.7101 1>,<at 0.9394 1>]}
       }})
     Scale Factor=&MakeFloat4FromVectorAndFloat(
      "Vector (RGB)"=&VectorFromFloat(Value=&FN_FrontEndFloatModifier(Gameplay=1, Front End=0.4))
      "Float (Alpha)"=1
     )
    )) {isSetParameters=true}
   /SetVariables(Particles.GlowSpriteColor:LinearColor=&MultiplyLinearColor(
     Linear Color=$Particles.SparkSpriteColor
     Scale Factor=&MakeFloat4FromVectorAndFloat("Vector (RGB)"=<1 1 1>, "Float (Alpha)"=0.05)
    )) {isSetParameters=true}
   /SetVariables(Particles.FlareSpriteColor:LinearColor=&MultiplyLinearColor(
     Linear Color=&ColorFromCurve(ColorCurve=DataInterface() {propertyValues {
        AlphaCurve {keys [<at 0.01 1>,<at 0.9198 1 RCIM_Cubic>,<at 1 0>]}
        BlueCurve {keys [<at 0 1.1>,<at 0.087 0>,<at 0.4906 0>,<at 0.7101 0>,<at 0.9394 0>]}
        GreenCurve {keys [<at 0 8.36>,<at 0.087 2.2>,<at 0.4906 1.0705>,<at 0.7101 0.1833>,<at 0.9394 0.0762>]}
        RedCurve {keys [<at 0 11>,<at 0.087 3>,<at 0.4906 2>,<at 0.7101 1>,<at 0.9394 1>]}
       }})
     Scale Factor=&MakeFloat4FromVectorAndFloat(
      "Vector (RGB)"=&VectorFromFloat(Value=&FN_FrontEndFloatModifier(Gameplay=1, Front End=0.75))
      "Float (Alpha)"=&FloatFromCurve(
       FloatCurve=DataInterface() {propertyValues {Curve {keys [
           {tangentMode=RCTM_User, time=0, value=1}
           <at 0.0756 -0.7485>
           <at 0.1284 2.2279>
           <at 0.3098 -0.7559>
           <at 0.4525 0.9206>
           <at 0.525 0.0265>
           <at 0.6661 0.3676>
           <at 0.7436 1.6548>
           <at 0.8248 0.1441>
           <at 0.9286 0.8147>
           {tangentMode=RCTM_User, time=1, value=0}
          ]}}}
       CurveIndex=&NormalizeFloat(
        Value=&Multiply_Float(A=$Particles.NormalizedAge, B=&RandomRangeFloat(Minimum=2.5, Maximum=4))
        Min=0
        Max=1
       )
       Scale Curve=0.1
      )
     )
    )) {isSetParameters=true}
   /SetVariables(Particles.LightColor=&MultiplyLinearColor(
     Linear Color=$Particles.FlareSpriteColor
     Scale Factor=&MakeFloat4FromVectorAndFloat("Vector (RGB)"=&VectorFromFloat(Value=3), "Float (Alpha)"=1)
    )) {isSetParameters=true}
   &Drag(
    Drag=&FloatFromCurve(
     FloatCurve=DataInterface() {propertyValues {Curve {keys [{interpMode=RCIM_Cubic, tangentMode=RCTM_User, time=0, value=1}, {interpMode=RCIM_Cubic, tangentMode=RCTM_User, time=1, value=0}]}}}
     Scale Curve=&Lerp_Float(A=17, B=5, Alpha=$Particles.PerParticleRandom)
    )
    Use Linear Drag=true
    Use Rotational Drag=true
   )
   &GravityForce(Gravity=&MaskVectorBySpawnGroup(Masked Vector=<0 0 -980>, Passthrough Vector=<0 0 0>))
   &InheritSourceMovement(
    Applied Position Scale=&VectorFromCurve(VectorCurve=DataInterface() {propertyValues {
       XCurve {keys [
         {arriveTangent=0.1831, interpMode=RCIM_Cubic, leaveTangent=0.1831, tangentMode=RCTM_User, time=0, value=1}
         {arriveTangent=-0.6074, interpMode=RCIM_Cubic, leaveTangent=-0.6074, tangentMode=RCTM_User, time=0.2, value=0.3}
         {arriveTangent=-0.2088, interpMode=RCIM_Cubic, leaveTangent=-0.2088, tangentMode=RCTM_User, time=1, value=0}
        ]}
       YCurve {keys [
         {arriveTangent=0.1831, interpMode=RCIM_Cubic, leaveTangent=0.1831, tangentMode=RCTM_User, time=0, value=1}
         {arriveTangent=-0.6074, interpMode=RCIM_Cubic, leaveTangent=-0.6074, tangentMode=RCTM_User, time=0.2, value=0.3}
         {arriveTangent=-0.2088, interpMode=RCIM_Cubic, leaveTangent=-0.2088, tangentMode=RCTM_User, time=1, value=0}
        ]}
       ZCurve {keys [
         {arriveTangent=0.1831, interpMode=RCIM_Cubic, leaveTangent=0.1831, tangentMode=RCTM_User, time=0, value=1}
         {arriveTangent=0.0668, interpMode=RCIM_Cubic, leaveTangent=0.0668, tangentMode=RCTM_User, time=0.75, value=1}
         {arriveTangent=-3.0473, interpMode=RCIM_Cubic, leaveTangent=-3.0473, tangentMode=RCTM_User, time=1, value=0}
        ]}
      }})
    ApplyForce=false
    ApplyPosition=true
    ApplyPositionFalloff=false
   )
   &SolveForcesAndVelocity(Clamp Velocity=false, Limit Acceleration=false)
   /SetVariables(Particles.VisibilityTag=&MakeCustomIntFromBool(
     Bool=&SetBoolByFloatComparison(A=$System.Timeline.NormalizedTime, B=0)
     True INT=999
     False INT=$Particles.VisibilityTag
    )) {isSetParameters=true}
  ]
  rendererClasses [&NiagaraSpriteRendererProperties, &NiagaraLightRendererProperties]
  renderers [
   {rendererClass=&NiagaraSpriteRendererProperties}
   {rendererClass=&NiagaraSpriteRendererProperties, rendererIndex=1}
   {rendererClass=&NiagaraLightRendererProperties, rendererIndex=2}
   {rendererClass=&NiagaraSpriteRendererProperties, rendererIndex=3}
  ]
 }
 Main_Trail {
  emitterSpawnScript [&FN_RibbonIDOffsetUtility()]
  emitterUpdateScript [
   &EmitterState(
    Inactive Response=::NewEnumerator1
    Life Cycle Mode=::NewEnumerator1
    Loop Behavior=::NewEnumerator1
    Loop Delay=0.05
    Loop Duration Mode=::NewEnumerator1
    UseLoopDelay=true
   )
   &FN_SpawnAnimTrail(Number of Anim Trails=2)
  ]
  particleSpawnScript [
   &FN_InitializeAnimTrailBoneIndecies() {module=FN_InitializeAnimTrailBoneIndecies001}
   &FN_InitializeAnimTrailBoneIndecies(AnimTrail Selection=1, First Bone or Socket Index=2, Second Bone or Socket Index=3)
   &FN_InitializeAnimTrail(AnimTrailWidth=2.5, Color=<rgba 0.75 0.75 0.75 0.25>, Lifetime=0.15, SkeletalMesh=$System.Skeletal Mesh)
   &AddVelocity(Velocity=&FN_FrontEndVector3Modifier(Gameplay=<0 500 150>, Front End=<0 -1500 0>))
   &AddVelocity(Rotation Coordinate Space=::World, Velocity=<0 0 -250>) {module=AddVelocity001}
   &CameraOffset(Camera Offset Amount=50)
  ]
  particleUpdateScript [
   &FN_UpdateAnimTrail(SkeletalMesh=$System.Skeletal Mesh)
   &InheritSourceMovement(
    Applied Position Scale=&MakeVector(
     X=&FN_FrontEndFloatModifier(
      Gameplay=&FloatFromCurve(FloatCurve=DataInterface() {propertyValues {Curve {keys [
           {interpMode=RCIM_Cubic, tangentMode=RCTM_User, time=0.8, value=1}
           {arriveTangent=-16.3348, interpMode=RCIM_Cubic, leaveTangent=-16.3348, tangentMode=RCTM_User, time=1, value=0}
          ]}}})
      Front End=1
     )
     Y=&FN_FrontEndFloatModifier(
      Gameplay=&FloatFromCurve(FloatCurve=DataInterface() {propertyValues {Curve {keys [
           {interpMode=RCIM_Cubic, tangentMode=RCTM_User, time=0.8, value=1}
           {arriveTangent=-16.3348, interpMode=RCIM_Cubic, leaveTangent=-16.3348, tangentMode=RCTM_User, time=1, value=0}
          ]}}})
      Front End=1
     )
     Z=1
    )
    ApplyForce=false
    ApplyPosition=true
    ApplyPositionFalloff=false
   )
   &SolveForcesAndVelocity(Clamp Velocity=false, Limit Acceleration=false)
   &ScaleColor(Scale Alpha=$System.Timeline003.NormalizedTime, ScaleA=true, ScaleRGB=true)
  ]
  rendererClasses [&NiagaraRibbonRendererProperties]
  renderers [{rendererClass=&NiagaraRibbonRendererProperties}]
 }
 Smoke {
  emitterUpdateScript [
   &EmitterState(Life Cycle Mode=::NewEnumerator1, Loop Behavior=::NewEnumerator1, Loop Delay=0.65, Loop Duration Mode=::NewEnumerator1, UseLoopDelay=true)
   &SpawnRate(Spawn Probability=$System.Timeline.NormalizedTime, SpawnRate=22.5, Use Spawn Probability=true)
  ]
  particleSpawnScript [
   &PartitionParticles(Number Of Partitions=3)
   &InitializeParticle(
    Color=&MakeLinearColorFromVectorAndFloat(
     "Vector (RGB)"=<1 1 1>
     "Float (Alpha)"=&Multiply_Float(
      A=&FN_FrontEndFloatModifier(Gameplay=1, Front End=0.5)
      B=&Lerp_Float(A=0.15, B=0.06, Alpha=$System.Timeline.NormalizedTime)
     )
    )
    Lifetime=0.333
    Position Offset=<0 88 -7>
    Sprite Rotation Mode=::NewEnumerator2
    Sprite Size Mode=::NewEnumerator3
    Sprite UV Mode=::NewEnumerator3
    Uniform Sprite Size=90
    UsePositionOffset=true
    Write Material Random=true
   )
   &SkeletalMeshLocation(
    Bone / Socket Index=1
    Bone Sampling Mode=::NewEnumerator4
    Interpolate Mesh Sampling=false
    MaskSpawnGroup=false
    Orientation Sampling=::NewEnumerator1
    Sampled Position Offset=/Niagara/DynamicInputs/Arrays/SelectPositionFromArray(
     Array Sampling Mode:ENiagaraArraySamplingMode=::NewEnumerator1
     Direct Array Index:NiagaraInt32=$Particles.PartitionParticles.Partition
     Interpolation:NiagaraFloat=?
     Direct Array Mode:ENiagara_WrapClamp=::NewEnumerator0
     Evaluation Type:ENiagaraRandomnessEvaluation=::NewEnumerator0
     Normalized Interpolation:NiagaraBool=true
     Recalculate Random Each Loop:NiagaraBool=false
     Position Array:NiagaraDataInterfaceArrayPosition=$System.Position Array
     Randomness Mode:ENiagaraRandomnessMode=::NewEnumerator0
     Fixed Random Seed:NiagaraBool=false
     Random Seed:NiagaraInt32=?
    )
    Skeletal Mesh=$System.Skeletal Mesh
   )
   &ApplyInitialForces()
   &AddVelocity(
    Cone Angle=1
    Cone Axis=<0 1 0.3>
    Module.UseConeFalloff=false
    Module.UseVelDistribution=false
    Override Seed=false
    Velocity Mode=::NewEnumerator2
    Velocity Speed=&RandomRangeFloat(Minimum=444, Maximum=666)
   )
  ]
  particleUpdateScript [
   &ParticleState()
   &ScaleSpriteSize(Uniform Curve Sprite Scale=DataInterface() {propertyValues {Curve {keys [
        {arriveTangent=2.5667, interpMode=RCIM_Cubic, leaveTangent=2.5667, tangentMode=RCTM_User, time=0, value=0}
        {arriveTangent=0.3152, interpMode=RCIM_Cubic, leaveTangent=0.3152, tangentMode=RCTM_User, time=1, value=1}
       ]}}})
   &DynamicMaterialParameters(
    Index 0 Param 1=&Add_Float(
     A=&Multiply_Float(A=$Particles.NormalizedAge, B=0.4)
     B=&RandomRangeFloat(Minimum=0, Maximum=1)
    )
    Module.Param0WriteEnabled=true
    Module.Param1WriteEnabled=false
    Module.Param2WriteEnabled=false
    Module.Param3WriteEnabled=false
   )
   &Drag(Drag=2, Use Linear Drag=true, Use Rotational Drag=false)
   &InheritSourceMovement(
    Applied Force Scale=<-0.5 -0.5 0>
    ApplyForce=true
    ApplyForceFalloff=false
    ApplyPosition=false
    Module.LimitForceSpeed=false
    Source Velocity:Vector3f=&TransformVector(Vector=$Engine.Owner.Velocity, Source Space=::World, Destination Space=::Simulation)
   )
   &SolveForcesAndVelocity(Clamp Velocity=false, Limit Acceleration=false)
   &ScaleColor(
    Scale Alpha=&FloatFromCurve(
     FloatCurve=DataInterface() {propertyValues {Curve {keys [
         {arriveTangent=6.4444, interpMode=RCIM_Cubic, leaveTangent=6.4444, tangentMode=RCTM_User, time=0, value=0}
         <at 0.3 0.85 RCIM_Cubic>
         {arriveTangent=-1.6012, interpMode=RCIM_Cubic, leaveTangent=-1.6012, tangentMode=RCTM_User, time=1, value=0}
        ]}}}
     Scale Curve=$System.Timeline.NormalizedTime
    )
    ScaleA=true
    ScaleRGB=false
   )
  ]
  rendererClasses [&NiagaraSpriteRendererProperties]
  renderers [{rendererClass=&NiagaraSpriteRendererProperties}]
 }
 Sparks {
  emitterUpdateScript [
   &EmitterState(Life Cycle Mode=::NewEnumerator1, Loop Behavior=::NewEnumerator1, Loop Delay=0.65, Loop Duration Mode=::NewEnumerator1, UseLoopDelay=true)
   &SpawnRate(
    Spawn Probability=$System.Timeline.NormalizedTime
    SpawnRate=&RandomRangeFloat(Minimum=50, Maximum=75)
    Use Spawn Probability=true
   )
  ]
  particleSpawnScript [
   &PartitionParticles(Number Of Partitions=3)
   &InitializeParticle(
    Color Mode=::NewEnumerator0
    Lifetime=&RandomRangeFloat(Minimum=0.066, Maximum=0.333)
    Position=&SelectVectorFromArray(
     Direct Array Index=&SpawnGroup()
     Vector Selection Array=DataInterface() {propertyValues {FloatData [<0 88 -7>, <0 -111 -7>]}}
    )
    Position Mode=::NewEnumerator1
    Sprite Size=&Lerp_Vector2(A=<3 5>, B=<1.5 20>, Alpha=&RandomRangeFloat(Minimum=0, Maximum=1))
    Sprite Size Mode=::NewEnumerator4
    UsePositionOffset=false
    Write Material Random=true
   )
   &SkeletalMeshLocation(
    Bone / Socket Index=1
    Bone Sampling Mode=::NewEnumerator4
    Interpolate Mesh Sampling=false
    MaskSpawnGroup=false
    Orientation Sampling=::NewEnumerator1
    Sampled Position Offset=&SelectPositionFromArray(
     Array Sampling Mode:ENiagaraArraySamplingMode=::NewEnumerator1
     Direct Array Index:NiagaraInt32=$Particles.PartitionParticles.Partition
     Interpolation:NiagaraFloat=?
     Direct Array Mode:ENiagara_WrapClamp=::NewEnumerator0
     Evaluation Type:ENiagaraRandomnessEvaluation=::NewEnumerator0
     Normalized Interpolation:NiagaraBool=true
     Recalculate Random Each Loop:NiagaraBool=false
     Position Array:NiagaraDataInterfaceArrayPosition=$System.Position Array
     Randomness Mode:ENiagaraRandomnessMode=::NewEnumerator0
     Fixed Random Seed:NiagaraBool=false
     Random Seed:NiagaraInt32=?
    )
    Skeletal Mesh=$System.Skeletal Mesh
   )
   &ShapeLocation(
    MaskSpawnGroup=false
    Override Seed=false
    Sphere Radius=&RandomRangeFloat(Minimum=10, Maximum=15)
    UseOwnerScale=true
   )
   &AddVelocity(
    Cone Angle=&RandomRangeFloat(Minimum=40, Maximum=120)
    Cone Axis=<0 1 0.3>
    Distribution Along Cone Axis:NiagaraFloat=&RandomRangeFloat(Minimum=0.5, Maximum=1)
    Module.UseConeFalloff=true
    Module.UseVelDistribution=true
    Override Seed=false
    Speed Falloff From Cone Axis:NiagaraFloat=&RandomRangeFloat(Minimum=0, Maximum=1)
    Velocity Mode=::NewEnumerator2
    Velocity Speed=&RandomRangeFloat(Minimum=444, Maximum=888)
   )
   /SetVariables(Particles.SubImageIndex=&RandomRangeFloat(Minimum=0, Maximum=4)) {isSetParameters=true}
  ]
  particleUpdateScript [
   &ParticleState()
   &Drag(Drag=&RandomRangeFloat(Minimum=1, Maximum=3), Use Linear Drag=true, Use Rotational Drag=false)
   &GravityForce(Gravity=<0 0 -981>)
   &InheritSourceMovement(
    Applied Force Scale=&RandomRangeVector(Minimum=<-2.5 -2.5 0>, Maximum=<-1 -1 0>)
    ApplyForce=true
    ApplyForceFalloff=false
    ApplyPosition=false
    Module.LimitForceSpeed=false
    Source Velocity:Vector3f=&TransformVector(Vector=$Engine.Owner.Velocity, Source Space=::World, Destination Space=::Simulation)
   )
   &SolveForcesAndVelocity(Clamp Velocity=false, Limit Acceleration=false)
   &ScaleSpriteSize(
    Non-Uniform Curve Sprite Scale=DataInterface() {propertyValues {
      XCurve {keys [{tangentMode=RCTM_User, time=0, value=1}, <at 0.6 1>, {tangentMode=RCTM_User, time=1, value=0}]}
      YCurve {keys [{tangentMode=RCTM_User, time=0, value=1}, <at 0.5 2>, {tangentMode=RCTM_User, time=1, value=0}]}
     }}
    Scale Sprite Size Mode=::NewEnumerator3
   )
   &Color(
    Color=&ColorFromCurve(ColorCurve=DataInterface() {propertyValues {
       AlphaCurve {keys [<at 0.01 0>, <at 0.0825 1>, {arriveTangent=-1.0899, interpMode=RCIM_Cubic, leaveTangent=-1.0899, time=0.9198, value=1}, <at 1 0>]}
       BlueCurve {keys [<at 0.0148 1.1>,<at 0.0618 0>,<at 0.1584 0>,<at 0.7101 0>,<at 0.9394 0>]}
       GreenCurve {keys [<at 0.0148 8.36>,<at 0.0618 2.2>,<at 0.1584 0.5>,<at 0.7101 0.1833>,<at 0.9394 0.0762>]}
       RedCurve {keys [<at 0.0148 11>,<at 0.0618 3>,<at 0.1584 2>,<at 0.7101 1>,<at 0.9394 1>]}
      }})
    Scale Alpha=&SelectFloatFromArray(Direct Array Index=&SpawnGroup(), Float Selection Array=DataInterface() {propertyValues {FloatData [1,5]}})
    Scale Color=&VectorFromFloat(Value=&SelectFloatFromArray(Direct Array Index=&SpawnGroup(), Float Selection Array=DataInterface() {propertyValues {FloatData [1,5]}}))
   )
   &ScaleSpriteSizeBySpeed(Max Scale Factor=<1 5>, Velocity Threshold=999)
  ]
  rendererClasses [&NiagaraSpriteRendererProperties]
  renderers [{rendererClass=&NiagaraSpriteRendererProperties}]
 }
 Sparks001 {
  emitterUpdateScript [
   &EmitterState(Life Cycle Mode=::NewEnumerator1, Loop Behavior=::NewEnumerator1, Loop Delay=0.65, Loop Duration Mode=::NewEnumerator1, UseLoopDelay=true)
   &SpawnBurst_Instantaneous(Spawn Count=3, Use Spawn Probability=false, UseLoopCountLimit=false)
  ]
  particleSpawnScript [
   &PartitionParticles(Number Of Partitions=3)
   &InitializeParticle(
    Color Mode=::NewEnumerator0
    Lifetime Mode=::NewEnumerator1
    Mesh Scale Mode=::NewEnumerator3
    Position=&SelectVectorFromArray(
     Direct Array Index=&SpawnGroup()
     Vector Selection Array=DataInterface() {propertyValues {FloatData [<0 88 -7>, <0 -111 -7>]}}
    )
    Position Mode=::NewEnumerator1
    UsePositionOffset=false
    Write Material Random=true
   )
   /SetVariables(Particles.Glow_SpriteSize:Vector2f=<15 15>) {isSetParameters=true}
   &CameraOffset(Camera Offset Amount=17.5)
  ]
  particleUpdateScript [
   &ParticleState(
    Kill Particles When Lifetime Has Elapsed=false
    Let Infinitely Lived Particles Die When Emitter Deactivates=true
    Loop Particles Lifetime=true
   )
   &SkeletalMeshLocation(
    Bone / Socket Index=1
    Bone Sampling Mode=::NewEnumerator4
    MaskSpawnGroup=false
    Orientation Sampling=::NewEnumerator1
    Sampled Position Coordinate Space=::Local
    Sampled Position Offset=&SelectPositionFromArray(
     Array Sampling Mode:ENiagaraArraySamplingMode=::NewEnumerator1
     Direct Array Index:NiagaraInt32=$Particles.PartitionParticles.Partition
     Interpolation:NiagaraFloat=?
     Direct Array Mode:ENiagara_WrapClamp=::NewEnumerator0
     Evaluation Type:ENiagaraRandomnessEvaluation=::NewEnumerator0
     Normalized Interpolation:NiagaraBool=true
     Recalculate Random Each Loop:NiagaraBool=false
     Position Array:NiagaraDataInterfaceArrayPosition=$System.Position Array
     Randomness Mode:ENiagaraRandomnessMode=::NewEnumerator0
     Fixed Random Seed:NiagaraBool=false
     Random Seed:NiagaraInt32=?
    )
    Skeletal Mesh=$System.Skeletal Mesh
   )
   /SetVariables(Particles.GlowSpriteColor:LinearColor=&MultiplyLinearColor(
     Linear Color=&ColorFromCurve(ColorCurve=DataInterface() {propertyValues {
        AlphaCurve {keys [<at 0.01 1>,<at 0.9198 1 RCIM_Cubic>,<at 1 0>]}
        BlueCurve {keys [<at 0 1.1>,<at 0.1098 0>,<at 0.2714 0>,<at 0.5479 0>,<at 0.9394 0>]}
        GreenCurve {keys [<at 0 8.36>,<at 0.1098 1.0705>,<at 0.2714 0.1833>,<at 0.5479 2.2>,<at 0.9394 0.0762>]}
        RedCurve {keys [<at 0 11>,<at 0.1098 2>,<at 0.2714 1>,<at 0.5479 3>,<at 0.9394 1>]}
       }})
     Scale Factor=&MakeFloat4FromVectorAndFloat(
      "Vector (RGB)"=&VectorFromFloat(Value=&FN_FrontEndFloatModifier(Gameplay=1.5, Front End=0.75))
      "Float (Alpha)"=&FloatFromCurve(
       FloatCurve=DataInterface() {propertyValues {Curve {keys [
           {tangentMode=RCTM_User, time=0, value=1}
           <at 0.0281 0.4665>
           <at 0.1284 2.2279>
           <at 0.302 0.8625>
           <at 0.4281 0.4665>
           <at 0.5172 0.889>
           <at 0.6582 1.2302>
           <at 0.7436 1.6548>
           <at 0.8169 1.0067>
           <at 0.9281 0.4665>
           {tangentMode=RCTM_User, time=0.9922, value=0.8625}
          ]}}}
       CurveIndex=&NormalizeFloat(
        Value=&Multiply_Float(A=$Particles.NormalizedAge, B=&RandomRangeFloat(Minimum=1.5, Maximum=1))
        Min=0
        Max=1
       )
       Scale Curve=&Multiply_Float(A=0.1, B=$System.Timeline.NormalizedTime)
      )
     )
    )) {isSetParameters=true}
   /SetVariables(Particles.LightColor=&MultiplyLinearColor(
     Linear Color=$Particles.GlowSpriteColor
     Scale Factor=&MakeFloat4FromVectorAndFloat("Vector (RGB)"=&VectorFromFloat(Value=5), "Float (Alpha)"=1)
    )) {isSetParameters=true}
  ]
  rendererClasses [
   &NiagaraSpriteRendererProperties
   &NiagaraLightRendererProperties
   &NiagaraMeshRendererProperties
  ]
  renderers [
   {rendererClass=&NiagaraSpriteRendererProperties}
   {rendererClass=&NiagaraLightRendererProperties, rendererIndex=1}
   {rendererClass=&NiagaraMeshRendererProperties, rendererIndex=2}
  ]
 }
 SpeedMotion {
  emitterUpdateScript [
   &EmitterState(
    Inactive Response=::NewEnumerator1
    Life Cycle Mode=::NewEnumerator1
    Loop Behavior=::NewEnumerator1
    Loop Delay=0.5
    Loop Duration Mode=::NewEnumerator1
    UseLoopDelay=true
   )
   &SpawnBurst_Instantaneous(Use Spawn Probability=false, UseLoopCountLimit=false)
  ]
  particleSpawnScript [
   &InitializeParticle(
    Color=<rgba 0.225 0.2217 0.25 0.25>
    Lifetime=0.3
    Mesh Scale=<1 1 1.25>
    Mesh Scale Mode=::NewEnumerator4
    UsePositionOffset=false
    Write Material Random=true
   )
   &InitialMeshOrientation(Mesh Axis To Orient=<0 0 1>, Mesh Orientation Mode=::NewEnumerator2, Use Rotation Vector=false)
   &CameraOffset(Camera Offset Amount=10)
  ]
  particleUpdateScript [
   &ParticleState(
    Kill Particles When Lifetime Has Elapsed=false
    Let Infinitely Lived Particles Die When Emitter Deactivates=true
    Loop Particles Lifetime=true
   )
   &SkeletalMeshLocation(Bone Sampling Mode=::NewEnumerator4, MaskSpawnGroup=false, Skeletal Mesh=$System.Skeletal Mesh)
   &ScaleColor(
    Scale Alpha=&Multiply_Float(
     A=1
     B=&FloatFromCurve(
      FloatCurve=DataInterface() {propertyValues {Curve {
         defaultValue=1
         keys [{interpMode=RCIM_Cubic, tangentMode=RCTM_User, time=0, value=1}, {interpMode=RCIM_Cubic, tangentMode=RCTM_User, time=1, value=2}]
        }}}
      CurveIndex=$System.Timeline.NormalizedTime
     )
    )
    ScaleA=true
    ScaleRGB=true
   )
  ]
  rendererClasses [&NiagaraMeshRendererProperties]
  renderers [{rendererClass=&NiagaraMeshRendererProperties, rendererIndex=0}]
 }
 Steam {
  emitterUpdateScript [
   &EmitterState(Life Cycle Mode=::NewEnumerator1, Loop Behavior=::NewEnumerator1, Loop Duration Mode=::NewEnumerator1, UseLoopDelay=false)
   &SpawnRate(
    Spawn Probability=&FN_FrontEndFloatModifier(Gameplay=1, Front End=0)
    SpawnRate=24
    Use Spawn Probability=true
   )
  ]
  particleSpawnScript [
   &PartitionParticles(Number Of Partitions=3)
   &InitializeParticle(
    Color=<rgba 0.75 0.75 0.75 0.5>
    Lifetime Max=1.5
    Lifetime Min=0.6
    Lifetime Mode=::NewEnumerator1
    Position Offset=<0 0 4>
    Sprite Rotation Mode=::NewEnumerator2
    Sprite Size Mode=::NewEnumerator1
    Uniform Sprite Size Max=50
    Uniform Sprite Size Min=25
    UsePositionOffset=true
    Write Material Random=true
   )
   &SkeletalMeshLocation(
    Bone / Socket Index=1
    Bone Sampling Mode=::NewEnumerator4
    Interpolate Mesh Sampling=false
    MaskSpawnGroup=false
    Orientation Sampling=::NewEnumerator1
    Sampled Position Offset=&SelectPositionFromArray(
     Array Sampling Mode:ENiagaraArraySamplingMode=::NewEnumerator1
     Direct Array Index:NiagaraInt32=$Particles.PartitionParticles.Partition
     Interpolation:NiagaraFloat=?
     Direct Array Mode:ENiagara_WrapClamp=::NewEnumerator0
     Evaluation Type:ENiagaraRandomnessEvaluation=::NewEnumerator0
     Normalized Interpolation:NiagaraBool=true
     Recalculate Random Each Loop:NiagaraBool=false
     Position Array:NiagaraDataInterfaceArrayPosition=$System.Position Array
     Randomness Mode:ENiagaraRandomnessMode=::NewEnumerator0
     Fixed Random Seed:NiagaraBool=false
     Random Seed:NiagaraInt32=?
    )
    Skeletal Mesh=$System.Skeletal Mesh
   )
   &AddVelocity(Velocity=&RandomRangeVector(Minimum=<-1 -1 10>, Maximum=<1 1 20>, Evaluation Type=::NewEnumerator1))
   &CameraOffset(Camera Offset Amount=10)
  ]
  particleUpdateScript [
   &ParticleState()
   &SubUVAnimation(
    Sub UV Play Rate:NiagaraFloat=&RandomRangeFloat(Minimum=0.05, Maximum=0.1)
    SubUV Animation Mode:ENiagara_SubUVLookupMode=::NewEnumerator3
    SubUV Start Time Offset:NiagaraFloat=&RandomRangeFloat(Minimum=0, Maximum=63)
   )
   &DynamicMaterialParameters(
    Index 0 Param 1=&FloatFromCurve(FloatCurve=DataInterface() {propertyValues {Curve {keys [<at 0 -0.25>,<at 1 -1>]}}})
    Index 0 Param 4=&OneMinusFloat(Float=$Particles.NormalizedAge)
    Module.Param0WriteEnabled=true
    Module.Param1WriteEnabled=false
    Module.Param2WriteEnabled=false
    Module.Param3WriteEnabled=true
   )
   &ScaleColor(
    Scale Alpha=&FloatFromCurve(
     FloatCurve=DataInterface() {propertyValues {Curve {keys [
         {arriveTangent=5.2427, interpMode=RCIM_Cubic, leaveTangent=5.2427, tangentMode=RCTM_User, time=0, value=0}
         {arriveTangent=-1.0898, interpMode=RCIM_Cubic, leaveTangent=-1.0898, tangentMode=RCTM_User, time=0.3, value=1}
         {arriveTangent=-0.1231, interpMode=RCIM_Cubic, leaveTangent=-0.1231, tangentMode=RCTM_User, time=1, value=0.5}
        ]}}}
     Scale Curve=&Multiply_Float(
      A=&Waveform(
       "Waveform [1]"=::NewEnumerator12
       "[1] Input Position"=&ReturnNormalizedExecIndex()
       "[1] Frequency"=&RandomRangeFloat(Minimum=1, Maximum=5)
       "[1] Amplitude Min/Max"=<0.5 1>
      )
      B=$System.Timeline.NormalizedTime
     )
    )
    ScaleA=true
    ScaleRGB=false
   )
   &ScaleSpriteSize(Scale Factor=&Vector2DFromFloat(Value=&FloatFromCurve(FloatCurve=DataInterface() {propertyValues {Curve {keys [
          {arriveTangent=1.8751, interpMode=RCIM_Cubic, leaveTangent=1.8751, tangentMode=RCTM_User, time=0, value=0.25}
          {arriveTangent=0.72, interpMode=RCIM_Cubic, leaveTangent=0.72, tangentMode=RCTM_User, time=1, value=1.25}
         ]}}})))
   &Drag(Use Linear Drag=true, Use Rotational Drag=false)
   &AccelerationForce(Acceleration=<0 0 111>)
   &InheritSourceMovement(
    Applied Force Scale=<-0.05 -0.05 0>
    ApplyForce=true
    ApplyForceFalloff=false
    ApplyPosition=false
    Module.LimitForceSpeed=false
    Source Velocity:Vector3f=&TransformVector(Vector=$Engine.Owner.Velocity, Source Space=::World, Destination Space=::Simulation)
   )
   &SolveForcesAndVelocity(Clamp Velocity=false, Limit Acceleration=false)
  ]
  rendererClasses [&NiagaraSpriteRendererProperties]
  renderers [{rendererClass=&NiagaraSpriteRendererProperties}]
 }
}
summary {
 systemName=NS_Glider_MuffinLadle_GasChin_Idle
 emitters [
  {emitterName=SpeedMotion, rendererClasses [&NiagaraMeshRendererProperties]}
  {emitterName=Sparks, rendererClasses [&NiagaraSpriteRendererProperties]}
  {emitterName=Smoke, rendererClasses [&NiagaraSpriteRendererProperties]}
  {emitterName=Main_Trail, rendererClasses [&NiagaraRibbonRendererProperties]}
  {
   emitterName=ElectricalSparksBurst
   rendererClasses [&NiagaraSpriteRendererProperties, &NiagaraLightRendererProperties]
  }
  {
   emitterName=Sparks001
   rendererClasses [
    &NiagaraSpriteRendererProperties
    &NiagaraLightRendererProperties
    &NiagaraMeshRendererProperties
   ]
  }
  {emitterName=Steam, rendererClasses [&NiagaraSpriteRendererProperties]}
 ]
 userVariables [
  User.bIsFrontEnd {type=NiagaraBool}
  User.bIsFrontEndPreview {type=NiagaraBool}
  User.bIsFullyDeployed {type=NiagaraBool}
  User.ForwardDot {type=NiagaraFloat, defaultValue=0, description=""}
  User.RibbonIDOffset {type=NiagaraInt32}
  User.RightDot {type=NiagaraFloat, defaultValue=0, description=""}
  User.SkeletalMesh {
   type=NiagaraDataInterfaceSkeletalMesh
   defaultValue {
    struct=NiagaraExt_VariableValue_DataInterface
    value {dataInterface=&self:NiagaraDataInterfaceSkeletalMesh_1}
   }
  }
 ]
}
